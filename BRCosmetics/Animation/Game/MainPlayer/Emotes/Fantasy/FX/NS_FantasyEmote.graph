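# Niagara effect for the Fantasy emote: short, timed bursts of blue-cyan glow and expanding ripples.
# Air_glow makes brief camera-offset glow sprites with alpha scaled by time of day, while MeshRipples002 drifts growing, randomly rolled mesh and sprite ripples along X.
# Follow-up bursts are gated by platform quality checks.
# Ripples also derive decal and refraction values and blend facing by view angle; two earlier ripple emitters are disabled.
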NiagaraSystem /BRCosmetics/Animation/Game/MainPlayer/Emotes/Fantasy/FX/NS_FantasyEmote
dependencies {
 usedDataInterfaces [
  &NiagaraDataInterfaceCurve
  &NiagaraDataInterfaceColorCurve
  &NiagaraDataInterfaceCamera
  &NiagaraDataInterfacePlatformSet
 ]
 usedDynamicInputs []
 usedModules [&SystemState]
}
emitters {
 Air_glow {
  emitterUpdateScript [
   &EmitterState()
   &SpawnBurst_Instantaneous(Use Spawn Probability=false, UseLoopCountLimit=false)
   &SpawnBurst_Instantaneous(
    Spawn Probability=&MakeCustomFloatFromBool(
     Bool=&IsPlatformSetActive(New Platform Set=DataInterface() {propertyValues {Platforms {qualityLevelMask=-31}}})
     True Float=0
     False Float=1
    )
    Spawn Time=0.05
    Use Spawn Probability=true
    UseLoopCountLimit=false
   ) {module=SpawnBurst_Instantaneous002}
   &SpawnBurst_Instantaneous(Spawn Time=0.35, Use Spawn Probability=false, UseLoopCountLimit=false) {module=SpawnBurst_Instantaneous001}
   &SpawnBurst_Instantaneous(
    Spawn Probability=&MakeCustomFloatFromBool(
     Bool=&IsPlatformSetActive(New Platform Set=DataInterface() {propertyValues {Platforms {qualityLevelMask=-31}}})
     True Float=0
     False Float=1
    )
    Spawn Time=0.425
    Use Spawn Probability=true
    UseLoopCountLimit=false
   ) {module=SpawnBurst_Instantaneous003}
  ]
  particleSpawnScript [
   &InitializeParticle(
    Color=&ColorFromCurve(
     ColorCurve=DataInterface() {propertyValues {
       AlphaCurve {keys [<at 0.5561 1>]}
       BlueCurve {keys [
         {interpMode=RCIM_Cubic, leaveTangent=3.2232, tangentMode=RCTM_Break, time=0, value=0.6688}
         {interpMode=RCIM_Cubic, tangentMode=RCTM_Break, time=0.1324, value=0.85}
         {arriveTangent=7.7104, interpMode=RCIM_Cubic, tangentMode=RCTM_Break, time=1, value=3}
        ]}
       GreenCurve {keys [
         <at 0 0.475 RCIM_Cubic>
         {
          arriveTangent=3.3365
          arriveTangentWeight=0.2322
          interpMode=RCIM_Cubic
          leaveTangent=-3.1167
          leaveTangentWeight=0.4094
          tangentMode=RCTM_Break
          tangentWeightMode=RCTWM_WeightedBoth
          time=0.1324
          value=0.7809
         }
         {
          arriveTangentWeight=0.6868
          interpMode=RCIM_Cubic
          leaveTangentWeight=0.6868
          tangentMode=RCTM_User
          tangentWeightMode=RCTWM_WeightedBoth
          time=1
          value=0.3063
         }
        ]}
       RedCurve {keys [
         <at 0 0.05 RCIM_Cubic>
         {
          arriveTangent=2.3842
          arriveTangentWeight=0.1724
          interpMode=RCIM_Cubic
          leaveTangent=-3.2096
          leaveTangentWeight=0.3729
          tangentMode=RCTM_Break
          tangentWeightMode=RCTWM_WeightedBoth
          time=0.1324
          value=0.2975
         }
         {
          arriveTangentWeight=0.6899
          interpMode=RCIM_Cubic
          leaveTangentWeight=0.6899
          tangentMode=RCTM_User
          tangentWeightMode=RCTWM_WeightedBoth
          time=1
          value=0.06
         }
        ]}
      }}
     CurveIndex=$Emitter.NormalizedLoopAge
    )
    Lifetime=0.175
    Sprite Size Mode=::NewEnumerator3
    Uniform Sprite Size=250
    UsePositionOffset=true
    Write Material Random=true
   )
   &InitializeParticle(
    Color=&ColorFromCurve(
     ColorCurve=DataInterface() {propertyValues {
       AlphaCurve {keys [<at 0.5561 1>]}
       BlueCurve {keys [
         {interpMode=RCIM_Cubic, leaveTangent=3.2232, tangentMode=RCTM_Break, time=0, value=0.6688}
         {interpMode=RCIM_Cubic, tangentMode=RCTM_Break, time=0.1324, value=0.85}
         {arriveTangent=7.7104, interpMode=RCIM_Cubic, tangentMode=RCTM_Break, time=1, value=3}
        ]}
       GreenCurve {keys [
         <at 0 0.475 RCIM_Cubic>
         {
          arriveTangent=3.3365
          arriveTangentWeight=0.2322
          interpMode=RCIM_Cubic
          leaveTangent=-3.1167
          leaveTangentWeight=0.4094
          tangentMode=RCTM_Break
          tangentWeightMode=RCTWM_WeightedBoth
          time=0.1324
          value=0.7809
         }
         {
          arriveTangentWeight=0.6868
          interpMode=RCIM_Cubic
          leaveTangentWeight=0.6868
          tangentMode=RCTM_User
          tangentWeightMode=RCTWM_WeightedBoth
          time=1
          value=0.3063
         }
        ]}
       RedCurve {keys [
         <at 0 0.05 RCIM_Cubic>
         {
          arriveTangent=2.3842
          arriveTangentWeight=0.1724
          interpMode=RCIM_Cubic
          leaveTangent=-3.2096
          leaveTangentWeight=0.3729
          tangentMode=RCTM_Break
          tangentWeightMode=RCTWM_WeightedBoth
          time=0.1324
          value=0.2975
         }
         {
          arriveTangentWeight=0.6899
          interpMode=RCIM_Cubic
          leaveTangentWeight=0.6899
          tangentMode=RCTM_User
          tangentWeightMode=RCTWM_WeightedBoth
          time=1
          value=0.06
         }
        ]}
      }}
     CurveIndex=$Emitter.NormalizedLoopAge
    )
    Lifetime=0.175
    Sprite Size Mode=::NewEnumerator3
    Uniform Sprite Size=250
    UsePositionOffset=true
    Write Material Random=true
   ) {enabled=false, module=InitializeParticle001}
   &CameraOffset(Camera Offset Amount=-50)
  ]
  particleUpdateScript [
   &ParticleState()
   &ScaleSpriteSize(
    Scale Sprite Size Mode=::NewEnumerator0
    Uniform Scale Factor=&FloatFromCurve(
     FloatCurve=DataInterface() {propertyValues {Curve {keys [<at 0 0.75>,<at 0.0263 1>,<at 0.08 0.75>,<at 0.1 1>,<at 0.154 0.75>,<at 0.2547 0.75>,<at 0.3 1>,<at 0.325 0.75>,<at 0.35 1>,<at 0.4242 0.75>]}}}
     CurveIndex=$Particles.Age
    )
   ) {enabled=false}
   &ScaleColor(
    Scale Alpha=&FloatFromCurve(
     FloatCurve=DataInterface() {propertyValues {Curve {keys [
         {arriveTangent=9.0725, interpMode=RCIM_Cubic, leaveTangent=9.0725, tangentMode=RCTM_User, time=0, value=0.5}
         {arriveTangent=-2.5249, interpMode=RCIM_Cubic, leaveTangent=-2.5249, tangentMode=RCTM_User, time=0.1408, value=0.9847}
         {arriveTangent=-0.0405, interpMode=RCIM_Cubic, leaveTangent=-0.0405, tangentMode=RCTM_User, time=1, value=0}
        ]}}}
     Scale Curve=&Multiply_Float(A=0.075, B=&FNTimeOfDayScalar(Night Scale=0.65, Morning Scale=1, Evening Scale=1))
    )
    ScaleA=true
    ScaleRGB=true
   )
  ]
  rendererClasses [&NiagaraSpriteRendererProperties]
  renderers [{rendererClass=&NiagaraSpriteRendererProperties}]
 }
 MeshRipples {
  bEnabled=false
  emitterUpdateScript [
   &EmitterState()
   &SpawnBurst_Instantaneous(Use Spawn Probability=false, UseLoopCountLimit=false)
   &SpawnBurst_Instantaneous(Spawn Time=0.05, Use Spawn Probability=false, UseLoopCountLimit=false) {module=SpawnBurst_Instantaneous002}
   &SpawnBurst_Instantaneous(Spawn Time=0.3, Use Spawn Probability=false, UseLoopCountLimit=false) {module=SpawnBurst_Instantaneous001}
   &SpawnBurst_Instantaneous(Spawn Time=0.375, Use Spawn Probability=false, UseLoopCountLimit=false) {module=SpawnBurst_Instantaneous003}
  ]
  particleSpawnScript [
   /SetVariables(Particles.Master Scale=1) {isSetParameters=true, module=SetVariables_#1}
   &InitializeParticle(
    Lifetime=0.35
    Mesh Scale Mode=::NewEnumerator3
    Mesh Uniform Scale=&Multiply_Float(A=1, B=$Particles.Master Scale)
    Position Offset=<10 -5 0>
    Position Offset Coordinate Space=::Local
    Sprite Rotation Mode=::NewEnumerator0
    Sprite Size Mode=::NewEnumerator3
    Uniform Sprite Size=&Multiply_Float(A=175, B=$Particles.Master Scale)
    UsePositionOffset=true
    Write Material Random=true
   )
   &AddVelocity(Velocity=<1 0 0>, Velocity Speed Scale=85)
   &InitialMeshOrientation(Mesh Orientation Mode=::NewEnumerator1, Rotation=<0 0.25 0>, Use Rotation Vector=true)
   /SetVariables(Particles.MeshOrientation=&MultiplyQuaternion(
     Quaternion A=&MakeQuatFromAxisAngle(
      Axis=&ReturnMeshOrientationAxis(Mesh Orientation Axis=::ZAxis, Mesh Orientation Quat=$Particles.MeshOrientation)
      AngleInDegrees=&RandomRangeFloat(Minimum=0, Maximum=360)
     )
     Quaternion B=$Particles.MeshOrientation
    )) {isSetParameters=true, module=SetVariables_#2}
   &CameraOffset()
  ]
  particleUpdateScript [
   &ParticleState()
   /SetVariables(Particles.RandScaleVector:Vector3f=&Lerp_Vector(
     A=<1 1 1>
     B=&RandomRangeVector(Minimum=<0.9 0.9 0.9>, Maximum=<1.1 1.1 1.1>, Evaluation Type=::NewEnumerator1)
     Alpha=0.75
    )) {isSetParameters=true, module=SetVariables_#3}
   &ScaleSpriteSize(
    Scale Sprite Size Mode=::NewEnumerator0
    Uniform Scale Factor=&FloatFromCurve(FloatCurve=DataInterface() {propertyValues {Curve {keys [<at 0 0.15 RCIM_Cubic>, {arriveTangent=1.0401, interpMode=RCIM_Cubic, leaveTangent=1.0401, tangentMode=RCTM_User, time=1, value=0.7}]}}})
   )
   &Drag(Drag=0.5, Use Linear Drag=true, Use Rotational Drag=true)
   &SolveForcesAndVelocity(Clamp Velocity=false, Limit Acceleration=false)
   &GenerateLocationEvent(Event Generation Enabled:NiagaraBool=$Transient.FirstFrame, Event Send Rate=120, Use Event Probability=false, UseEventDelay=false)
   &AlignSpriteToMeshOrientation(Mesh Orientation Relative Sprite Facing Vector:Vector3f=<0 0 1>, SetAlignment=true, SetFacing=true)
   /SetVariables(Particles.Mesh Orient Facing:Vector3f=$Particles.SpriteFacing) {isSetParameters=true, module=SetVariables_#4}
   &ScaleMeshSize(Scale Factor=&VectorFromFloat(Value=&FloatFromCurve(FloatCurve=DataInterface() {propertyValues {Curve {keys [<at 0 0.15 RCIM_Cubic>, {arriveTangent=1.0401, interpMode=RCIM_Cubic, leaveTangent=1.0401, tangentMode=RCTM_User, time=1, value=0.7}]}}})))
   &ScaleMeshSize(Scale Factor=$Particles.RandScaleVector) {module=ScaleMeshSize001}
   &ScaleColor(
    Scale Alpha=&FloatFromCurve(
     FloatCurve=DataInterface() {propertyValues {Curve {keys [
         {arriveTangent=-1.1111, interpMode=RCIM_Cubic, time=0, value=0}
         <at 0.1 1 RCIM_Cubic>
         {interpMode=RCIM_Cubic, leaveTangent=-3.1172, time=1, value=0}
        ]}}}
     Scale Curve=0.35
    )
    Scale RGB=&Lerp_Vector(
     A=&MakeVectorFromLinearColorRGB(Color=&ColorFromCurve(ColorCurve=DataInterface() {propertyValues {
         AlphaCurve {keys []}
         BlueCurve {keys [
           {interpMode=RCIM_Cubic, leaveTangent=3.2232, tangentMode=RCTM_Break, time=0, value=0.6688}
           {interpMode=RCIM_Cubic, tangentMode=RCTM_Break, time=0.1324, value=0.85}
           {arriveTangent=7.7104, interpMode=RCIM_Cubic, tangentMode=RCTM_Break, time=1, value=3}
          ]}
         GreenCurve {keys [
           <at 0 0.475 RCIM_Cubic>
           {
            arriveTangent=3.3365
            arriveTangentWeight=0.2322
            interpMode=RCIM_Cubic
            leaveTangent=-3.1167
            leaveTangentWeight=0.4094
            tangentMode=RCTM_Break
            tangentWeightMode=RCTWM_WeightedBoth
            time=0.1324
            value=0.7809
           }
           {
            arriveTangentWeight=0.6868
            interpMode=RCIM_Cubic
            leaveTangentWeight=0.6868
            tangentMode=RCTM_User
            tangentWeightMode=RCTWM_WeightedBoth
            time=1
            value=0.3063
           }
          ]}
         RedCurve {keys [
           <at 0 0.05 RCIM_Cubic>
           {
            arriveTangent=2.3842
            arriveTangentWeight=0.1724
            interpMode=RCIM_Cubic
            leaveTangent=-3.2096
            leaveTangentWeight=0.3729
            tangentMode=RCTM_Break
            tangentWeightMode=RCTWM_WeightedBoth
            time=0.1324
            value=0.2975
           }
           {
            arriveTangentWeight=0.6899
            interpMode=RCIM_Cubic
            leaveTangentWeight=0.6899
            tangentMode=RCTM_User
            tangentWeightMode=RCTWM_WeightedBoth
            time=1
            value=0.06
           }
          ]}
        }}))
     B=&MakeVectorFromLinearColorRGB(Color=&ColorFromCurve(ColorCurve=DataInterface() {propertyValues {
         AlphaCurve {keys []}
         BlueCurve {keys [
           {interpMode=RCIM_Cubic, leaveTangent=3.2232, tangentMode=RCTM_Break, time=0, value=0.6688}
           {interpMode=RCIM_Cubic, tangentMode=RCTM_Break, time=0.1324, value=0.85}
           {arriveTangent=7.7104, interpMode=RCIM_Cubic, tangentMode=RCTM_Break, time=1, value=3}
          ]}
         GreenCurve {keys [
           <at 0 0.475 RCIM_Cubic>
           {
            arriveTangent=3.3365
            arriveTangentWeight=0.2322
            interpMode=RCIM_Cubic
            leaveTangent=-3.1167
            leaveTangentWeight=0.4094
            tangentMode=RCTM_Break
            tangentWeightMode=RCTWM_WeightedBoth
            time=0.1324
            value=0.7809
           }
           {
            arriveTangentWeight=0.6868
            interpMode=RCIM_Cubic
            leaveTangentWeight=0.6868
            tangentMode=RCTM_User
            tangentWeightMode=RCTWM_WeightedBoth
            time=1
            value=0.2513
           }
          ]}
         RedCurve {keys [
           <at 0 0.05 RCIM_Cubic>
           {
            arriveTangent=2.3842
            arriveTangentWeight=0.1724
            interpMode=RCIM_Cubic
            leaveTangent=-3.2096
            leaveTangentWeight=0.3729
            tangentMode=RCTM_Break
            tangentWeightMode=RCTWM_WeightedBoth
            time=0.1324
            value=0.2975
           }
           {
            arriveTangentWeight=0.6899
            interpMode=RCIM_Cubic
            leaveTangentWeight=0.6899
            tangentMode=RCTM_User
            tangentWeightMode=RCTWM_WeightedBoth
            time=1
            value=0
           }
          ]}
        }}))
     Alpha=&RandomRangeFloat(Minimum=0, Maximum=1)
    )
    ScaleA=true
    ScaleRGB=true
   ) {enabled=false, module=ScaleColor001}
   &ScaleColor(
    Scale Alpha=&FloatFromCurve(
     FloatCurve=DataInterface() {propertyValues {Curve {keys [
         {arriveTangent=-1.1111, interpMode=RCIM_Cubic, time=0, value=0}
         <at 0.1 1 RCIM_Cubic>
         {interpMode=RCIM_Cubic, leaveTangent=-3.1172, time=1, value=0}
        ]}}}
     Scale Curve=0.35
    )
    Scale RGB=&MakeVectorFromLinearColorRGB(Color=&ColorFromCurve(ColorCurve=DataInterface() {propertyValues {
        AlphaCurve {keys []}
        BlueCurve {keys [
          {interpMode=RCIM_Cubic, leaveTangent=3.2232, tangentMode=RCTM_Break, time=0, value=0.6688}
          {interpMode=RCIM_Cubic, tangentMode=RCTM_Break, time=0.1324, value=1}
          <at 1 3 RCIM_Cubic>
         ]}
        GreenCurve {keys [
          <at 0 0.475 RCIM_Cubic>
          {
           arriveTangent=3.3365
           arriveTangentWeight=0.2322
           interpMode=RCIM_Cubic
           leaveTangent=-3.1167
           leaveTangentWeight=0.4094
           tangentMode=RCTM_Break
           tangentWeightMode=RCTWM_WeightedBoth
           time=0.1324
           value=0.9562
          }
          {
           arriveTangentWeight=0.6868
           interpMode=RCIM_Cubic
           leaveTangentWeight=0.6868
           tangentMode=RCTM_User
           tangentWeightMode=RCTWM_WeightedBoth
           time=1
           value=0.3063
          }
         ]}
        RedCurve {keys [
          <at 0 0.05 RCIM_Cubic>
          {
           arriveTangent=2.3842
           arriveTangentWeight=0.1724
           interpMode=RCIM_Cubic
           leaveTangent=-3.2096
           leaveTangentWeight=0.3729
           tangentMode=RCTM_Break
           tangentWeightMode=RCTWM_WeightedBoth
           time=0.1324
           value=0.65
          }
          {
           arriveTangentWeight=0.6899
           interpMode=RCIM_Cubic
           leaveTangentWeight=0.6899
           tangentMode=RCTM_User
           tangentWeightMode=RCTWM_WeightedBoth
           time=1
           value=0.06
          }
         ]}
       }}))
    ScaleA=true
    ScaleRGB=true
   )
   &DynamicMaterialParameters(
    Index 0 Param 4=&FloatFromCurve(
     FloatCurve=DataInterface() {propertyValues {Curve {keys [
         {
          arriveTangent=9.177
          arriveTangentWeight=0.6327
          interpMode=RCIM_Cubic
          leaveTangent=9.177
          leaveTangentWeight=0.6327
          tangentMode=RCTM_User
          tangentWeightMode=RCTWM_WeightedBoth
          time=0
          value=0.2
         }
         {
          arriveTangentWeight=0.6785
          interpMode=RCIM_Cubic
          leaveTangentWeight=0.6785
          tangentMode=RCTM_User
          tangentWeightMode=RCTWM_WeightedBoth
          time=1
          value=1
         }
        ]}}}
     Scale Curve=0.28
    )
    Module.Index0Param0Write=true
    Module.Index0Param1Write=true
    Module.Index0Param2Write=true
    Module.Index0Param3Write=true
   )
   &SpriteFacingAndAlignment(
    Facing Coordinate Space=::Local
    Sprite Alignment=&RandomVector(Vector Scale=1)
    Sprite Facing=&ReturnCameraVector(Camera Vectors=::NewEnumerator4, Coordinate Space=::Simulation, Origin Position=$Particles.Position)
    WriteAlignment=false
    WriteFacing=true
   )
   /SetVariables(Particles.Decal Color:LinearColor=&MultiplyLinearColor(
     Linear Color=$Particles.Color
     Scale Factor=&MakeFloat4FromVectorAndFloat("Vector (RGB)"=&VectorFromFloat(Value=2), "Float (Alpha)"=0.35)
    )) {isSetParameters=true, module=SetVariables_#5}
   /SetVariables(
    Particles.Refracl Color:LinearColor=&MultiplyLinearColor(
     Linear Color=$Particles.Color
     Scale Factor=&MakeFloat4FromVectorAndFloat("Vector (RGB)"=&VectorFromFloat(Value=5), "Float (Alpha)"=0)
    )
    Particles.refract dynamic:Vector4f=&MakeVector4(
     X=&Multiply_Float(
      A=&MakeFloatFromLinearColor(LinearColor=$Particles.Color, Channel:ENiagaraLinearColor_Channels=::NewEnumerator3)
      B=0.0025
     )
     Y=0
     Z=0
     W=0
    )
   ) {isSetParameters=true, module=SetVariables_#6}
   /SetVariables(Particles.RegularFacing:Vector3f=$Particles.SpriteFacing) {isSetParameters=true, module=SetVariables_#7}
   /SetVariables(Particles.NewFacing:Vector3f=&Lerp_Vector(
     A=$Particles.Mesh Orient Facing
     B=$Particles.RegularFacing
     Alpha=&FloatFromCurve(
      FloatCurve=DataInterface() {propertyValues {Curve {keys [<at -1 0 RCIM_Cubic>,<at 0 1 RCIM_Cubic>,<at 1 0 RCIM_Cubic>]}}}
      CurveIndex=&DotProduct(
       Vector A=&ReturnCameraVector(Camera Vectors=::NewEnumerator4, Coordinate Space=::Simulation, Origin Position=$Particles.Position)
       Vector B=$Engine.Owner.SystemXAxis
      )
     )
    )) {isSetParameters=true, module=SetVariables_#8}
   /SetVariables(Particles.NewFacing:Vector3f=$Particles.RegularFacing) {isSetParameters=true, module=SetVariables_#9}
   /SetVariables(Particles.VibratePosition:NiagaraPosition=&AddVectorToPosition(
     Position=$Particles.Position
     Vector=&Multiply_Vector(
      A=&Add_Vector(
       A=&Multiply_Vector(
        A=/Game/Effects/Niagara/DynamicInputs/Engine/FN_GetSystemAxis(Axis:ENiagaraOrientationAxis=::XAxis, Flip Axis:NiagaraBool=false)
        B=&VectorFromFloat(Value=&RandomRangeFloat(Minimum=-1, Maximum=-1, Evaluation Type=::NewEnumerator1))
       )
       B=&Multiply_Vector(
        A=&ReturnMeshOrientationAxis(Mesh Orientation Axis=::YAxis, Mesh Orientation Quat=$Particles.MeshOrientation)
        B=&VectorFromFloat(Value=&RandomRangeFloat(Minimum=-1, Maximum=1, Evaluation Type=::NewEnumerator1))
       )
      )
      B=&VectorFromFloat(Value=1)
     )
    )) {isSetParameters=true, module=SetVariables_#10}
  ]
  rendererClasses [&NiagaraMeshRendererProperties, &NiagaraSpriteRendererProperties]
  renderers [
   {rendererClass=&NiagaraMeshRendererProperties, rendererIndex=0}
   {rendererClass=&NiagaraMeshRendererProperties, rendererIndex=1}
   {rendererClass=&NiagaraSpriteRendererProperties, rendererIndex=2}
   {rendererClass=&NiagaraMeshRendererProperties, rendererIndex=3}
  ]
 }
 MeshRipples001 {
  bEnabled=false
  emitterUpdateScript [
   &EmitterState()
   &SpawnBurst_Instantaneous(Use Spawn Probability=false, UseLoopCountLimit=false)
   &SpawnBurst_Instantaneous(
    Spawn Probability=&MakeCustomFloatFromBool(
     Bool=&IsPlatformSetActive(New Platform Set=DataInterface() {propertyValues {Platforms {qualityLevelMask=-31}}})
     True Float=0
     False Float=1
    )
    Spawn Time=0.05
    Use Spawn Probability=true
    UseLoopCountLimit=false
   ) {module=SpawnBurst_Instantaneous002}
   &SpawnBurst_Instantaneous(Spawn Time=0.35, Use Spawn Probability=false, UseLoopCountLimit=false) {module=SpawnBurst_Instantaneous001}
   &SpawnBurst_Instantaneous(
    Spawn Probability=&MakeCustomFloatFromBool(
     Bool=&IsPlatformSetActive(New Platform Set=DataInterface() {propertyValues {Platforms {qualityLevelMask=-31}}})
     True Float=0
     False Float=1
    )
    Spawn Time=0.425
    Use Spawn Probability=true
    UseLoopCountLimit=false
   ) {module=SpawnBurst_Instantaneous003}
  ]
  particleSpawnScript [
   /SetVariables(Particles.Master Scale=1) {isSetParameters=true, module=SetVariables_#1}
   &InitializeParticle(
    Lifetime=0.325
    Mesh Scale Mode=::NewEnumerator3
    Mesh Uniform Scale=&Multiply_Float(A=1, B=$Particles.Master Scale)
    Position Offset=<10 -5 0>
    Position Offset Coordinate Space=::Local
    Sprite Rotation Mode=::NewEnumerator0
    Sprite Size Mode=::NewEnumerator3
    Uniform Sprite Size=&Multiply_Float(A=175, B=$Particles.Master Scale)
    UsePositionOffset=true
    Write Material Random=true
   )
   &AddVelocity(Velocity=<1 0 0>, Velocity Speed Scale=90)
   &InitialMeshOrientation(Mesh Orientation Mode=::NewEnumerator1, Rotation=<0 0.25 0>, Use Rotation Vector=true)
   /SetVariables(Particles.MeshOrientation=&MultiplyQuaternion(
     Quaternion A=&MakeQuatFromAxisAngle(
      Axis=&ReturnMeshOrientationAxis(Mesh Orientation Axis=::ZAxis, Mesh Orientation Quat=$Particles.MeshOrientation)
      AngleInDegrees=&RandomRangeFloat(Minimum=0, Maximum=360)
     )
     Quaternion B=$Particles.MeshOrientation
    )) {isSetParameters=true, module=SetVariables_#2}
   &CameraOffset()
  ]
  particleUpdateScript [
   &ParticleState()
   /SetVariables(Particles.RandScaleVector:Vector3f=&Lerp_Vector(
     A=<1 1 1>
     B=&RandomRangeVector(Minimum=<0.9 0.9 0.9>, Maximum=<1.1 1.1 1.1>, Evaluation Type=::NewEnumerator1)
     Alpha=0.75
    )) {isSetParameters=true, module=SetVariables_#3}
   &ScaleSpriteSize(
    Scale Sprite Size Mode=::NewEnumerator0
    Uniform Scale Factor=&FloatFromCurve(FloatCurve=DataInterface() {propertyValues {Curve {keys [<at 0 0.15 RCIM_Cubic>, {arriveTangent=1.0401, interpMode=RCIM_Cubic, leaveTangent=1.0401, tangentMode=RCTM_User, time=1, value=0.7}]}}})
   )
   &Drag(Drag=0.5, Use Linear Drag=true, Use Rotational Drag=true)
   &SolveForcesAndVelocity(Clamp Velocity=false, Limit Acceleration=false)
   &GenerateLocationEvent(Event Generation Enabled:NiagaraBool=$Transient.FirstFrame, Event Send Rate=120, Use Event Probability=false, UseEventDelay=false)
   &AlignSpriteToMeshOrientation(Mesh Orientation Relative Sprite Facing Vector:Vector3f=<0 0 1>, SetAlignment=true, SetFacing=true)
   /SetVariables(Particles.Mesh Orient Facing:Vector3f=$Particles.SpriteFacing) {isSetParameters=true, module=SetVariables_#4}
   &ScaleMeshSize(Scale Factor=&VectorFromFloat(Value=&FloatFromCurve(FloatCurve=DataInterface() {propertyValues {Curve {keys [
          {
           arriveTangent=0.7941
           arriveTangentWeight=0.3721
           interpMode=RCIM_Cubic
           leaveTangent=0.7941
           leaveTangentWeight=0.3721
           tangentMode=RCTM_User
           tangentWeightMode=RCTWM_WeightedBoth
           time=0.05
           value=0.15
          }
          {
           arriveTangentWeight=0.547
           interpMode=RCIM_Cubic
           leaveTangentWeight=0.547
           tangentMode=RCTM_User
           tangentWeightMode=RCTWM_WeightedBoth
           time=1
           value=0.7
          }
         ]}}})))
   &ScaleMeshSize(Scale Factor=$Particles.RandScaleVector) {module=ScaleMeshSize001}
   &ScaleColor(
    Scale Alpha=&FloatFromCurve(
     FloatCurve=DataInterface() {propertyValues {Curve {keys [
         {arriveTangent=-1.1111, interpMode=RCIM_Cubic, time=0, value=0}
         <at 0.1 1 RCIM_Cubic>
         {interpMode=RCIM_Cubic, leaveTangent=-3.1172, time=1, value=0}
        ]}}}
     Scale Curve=0.35
    )
    Scale RGB=&Lerp_Vector(
     A=&MakeVectorFromLinearColorRGB(Color=&ColorFromCurve(ColorCurve=DataInterface() {propertyValues {
         AlphaCurve {keys []}
         BlueCurve {keys [
           {interpMode=RCIM_Cubic, leaveTangent=3.2232, tangentMode=RCTM_Break, time=0, value=0.6688}
           {interpMode=RCIM_Cubic, tangentMode=RCTM_Break, time=0.1324, value=0.85}
           {arriveTangent=7.7104, interpMode=RCIM_Cubic, tangentMode=RCTM_Break, time=1, value=3}
          ]}
         GreenCurve {keys [
           <at 0 0.475 RCIM_Cubic>
           {
            arriveTangent=3.3365
            arriveTangentWeight=0.2322
            interpMode=RCIM_Cubic
            leaveTangent=-3.1167
            leaveTangentWeight=0.4094
            tangentMode=RCTM_Break
            tangentWeightMode=RCTWM_WeightedBoth
            time=0.1324
            value=0.7809
           }
           {
            arriveTangentWeight=0.6868
            interpMode=RCIM_Cubic
            leaveTangentWeight=0.6868
            tangentMode=RCTM_User
            tangentWeightMode=RCTWM_WeightedBoth
            time=1
            value=0.3063
           }
          ]}
         RedCurve {keys [
           <at 0 0.05 RCIM_Cubic>
           {
            arriveTangent=2.3842
            arriveTangentWeight=0.1724
            interpMode=RCIM_Cubic
            leaveTangent=-3.2096
            leaveTangentWeight=0.3729
            tangentMode=RCTM_Break
            tangentWeightMode=RCTWM_WeightedBoth
            time=0.1324
            value=0.2975
           }
           {
            arriveTangentWeight=0.6899
            interpMode=RCIM_Cubic
            leaveTangentWeight=0.6899
            tangentMode=RCTM_User
            tangentWeightMode=RCTWM_WeightedBoth
            time=1
            value=0.06
           }
          ]}
        }}))
     B=&MakeVectorFromLinearColorRGB(Color=&ColorFromCurve(ColorCurve=DataInterface() {propertyValues {
         AlphaCurve {keys []}
         BlueCurve {keys [
           {interpMode=RCIM_Cubic, leaveTangent=3.2232, tangentMode=RCTM_Break, time=0, value=0.6688}
           {interpMode=RCIM_Cubic, tangentMode=RCTM_Break, time=0.1324, value=0.85}
           {arriveTangent=7.7104, interpMode=RCIM_Cubic, tangentMode=RCTM_Break, time=1, value=3}
          ]}
         GreenCurve {keys [
           <at 0 0.475 RCIM_Cubic>
           {
            arriveTangent=3.3365
            arriveTangentWeight=0.2322
            interpMode=RCIM_Cubic
            leaveTangent=-3.1167
            leaveTangentWeight=0.4094
            tangentMode=RCTM_Break
            tangentWeightMode=RCTWM_WeightedBoth
            time=0.1324
            value=0.7809
           }
           {
            arriveTangentWeight=0.6868
            interpMode=RCIM_Cubic
            leaveTangentWeight=0.6868
            tangentMode=RCTM_User
            tangentWeightMode=RCTWM_WeightedBoth
            time=1
            value=0.2513
           }
          ]}
         RedCurve {keys [
           <at 0 0.05 RCIM_Cubic>
           {
            arriveTangent=2.3842
            arriveTangentWeight=0.1724
            interpMode=RCIM_Cubic
            leaveTangent=-3.2096
            leaveTangentWeight=0.3729
            tangentMode=RCTM_Break
            tangentWeightMode=RCTWM_WeightedBoth
            time=0.1324
            value=0.2975
           }
           {
            arriveTangentWeight=0.6899
            interpMode=RCIM_Cubic
            leaveTangentWeight=0.6899
            tangentMode=RCTM_User
            tangentWeightMode=RCTWM_WeightedBoth
            time=1
            value=0
           }
          ]}
        }}))
     Alpha=&RandomRangeFloat(Minimum=0, Maximum=1)
    )
    ScaleA=true
    ScaleRGB=true
   ) {enabled=false, module=ScaleColor001}
   &ScaleColor(
    Scale Alpha=&FloatFromCurve(
     FloatCurve=DataInterface() {propertyValues {Curve {keys [
         {arriveTangent=-1.1111, interpMode=RCIM_Cubic, time=0, value=0}
         <at 0.1 1 RCIM_Cubic>
         {interpMode=RCIM_Cubic, leaveTangent=-3.1172, time=1, value=0}
        ]}}}
     Scale Curve=0.35
    )
    Scale RGB=&MakeVectorFromLinearColorRGB(Color=&ColorFromCurve(ColorCurve=DataInterface() {propertyValues {
        AlphaCurve {keys []}
        BlueCurve {keys [
          {interpMode=RCIM_Cubic, leaveTangent=3.2232, tangentMode=RCTM_Break, time=0, value=0.6688}
          {interpMode=RCIM_Cubic, tangentMode=RCTM_Break, time=0.1324, value=1}
          <at 1 3 RCIM_Cubic>
         ]}
        GreenCurve {keys [
          <at 0 0.475 RCIM_Cubic>
          {
           arriveTangent=3.3365
           arriveTangentWeight=0.2322
           interpMode=RCIM_Cubic
           leaveTangent=-3.1167
           leaveTangentWeight=0.4094
           tangentMode=RCTM_Break
           tangentWeightMode=RCTWM_WeightedBoth
           time=0.1324
           value=0.9562
          }
          {
           arriveTangentWeight=0.6868
           interpMode=RCIM_Cubic
           leaveTangentWeight=0.6868
           tangentMode=RCTM_User
           tangentWeightMode=RCTWM_WeightedBoth
           time=1
           value=0.3063
          }
         ]}
        RedCurve {keys [
          <at 0 0.05 RCIM_Cubic>
          {
           arriveTangent=2.3842
           arriveTangentWeight=0.1724
           interpMode=RCIM_Cubic
           leaveTangent=-3.2096
           leaveTangentWeight=0.3729
           tangentMode=RCTM_Break
           tangentWeightMode=RCTWM_WeightedBoth
           time=0.1324
           value=0.65
          }
          {
           arriveTangentWeight=0.6899
           interpMode=RCIM_Cubic
           leaveTangentWeight=0.6899
           tangentMode=RCTM_User
           tangentWeightMode=RCTWM_WeightedBoth
           time=1
           value=0.06
          }
         ]}
       }}))
    ScaleA=true
    ScaleRGB=true
   )
   &DynamicMaterialParameters(
    Index 0 Param 4=&FloatFromCurve(
     FloatCurve=DataInterface() {propertyValues {Curve {keys [
         {
          arriveTangent=9.177
          arriveTangentWeight=0.6327
          interpMode=RCIM_Cubic
          leaveTangent=9.177
          leaveTangentWeight=0.6327
          tangentMode=RCTM_User
          tangentWeightMode=RCTWM_WeightedBoth
          time=0
          value=0.2
         }
         {
          arriveTangentWeight=0.6785
          interpMode=RCIM_Cubic
          leaveTangentWeight=0.6785
          tangentMode=RCTM_User
          tangentWeightMode=RCTWM_WeightedBoth
          time=1
          value=1
         }
        ]}}}
     Scale Curve=0.28
    )
    Module.Index0Param0Write=true
    Module.Index0Param1Write=true
    Module.Index0Param2Write=true
    Module.Index0Param3Write=true
   )
   &SpriteFacingAndAlignment(
    Facing Coordinate Space=::Local
    Sprite Alignment=&RandomVector(Vector Scale=1)
    Sprite Facing=&ReturnCameraVector(Camera Vectors=::NewEnumerator4, Coordinate Space=::Simulation, Origin Position=$Particles.Position)
    WriteAlignment=false
    WriteFacing=true
   )
   /SetVariables(Particles.Decal Color:LinearColor=&MultiplyLinearColor(
     Linear Color=$Particles.Color
     Scale Factor=&MakeFloat4FromVectorAndFloat("Vector (RGB)"=&VectorFromFloat(Value=2), "Float (Alpha)"=0.35)
    )) {isSetParameters=true, module=SetVariables_#5}
   /SetVariables(
    Particles.Refracl Color:LinearColor=&MultiplyLinearColor(
     Linear Color=$Particles.Color
     Scale Factor=&MakeFloat4FromVectorAndFloat("Vector (RGB)"=&VectorFromFloat(Value=5), "Float (Alpha)"=0)
    )
    Particles.refract dynamic:Vector4f=&MakeVector4(
     X=&Multiply_Float(
      A=&MakeFloatFromLinearColor(LinearColor=$Particles.Color, Channel:ENiagaraLinearColor_Channels=::NewEnumerator3)
      B=0.0025
     )
     Y=0
     Z=0
     W=0
    )
   ) {isSetParameters=true, module=SetVariables_#6}
   /SetVariables(Particles.RegularFacing:Vector3f=$Particles.SpriteFacing) {isSetParameters=true, module=SetVariables_#7}
   /SetVariables(Particles.NewFacing:Vector3f=&Lerp_Vector(
     A=$Particles.Mesh Orient Facing
     B=$Particles.RegularFacing
     Alpha=&FloatFromCurve(
      FloatCurve=DataInterface() {propertyValues {Curve {keys [<at -1 0 RCIM_Cubic>,<at 0 1 RCIM_Cubic>,<at 1 0 RCIM_Cubic>]}}}
      CurveIndex=&DotProduct(
       Vector A=&ReturnCameraVector(Camera Vectors=::NewEnumerator4, Coordinate Space=::Simulation, Origin Position=$Particles.Position)
       Vector B=$Engine.Owner.SystemXAxis
      )
     )
    )) {isSetParameters=true, module=SetVariables_#8}
   /SetVariables(Particles.NewFacing:Vector3f=$Particles.RegularFacing) {enabled=false, isSetParameters=true, module=SetVariables_#9}
   /SetVariables(Particles.VibratePosition:NiagaraPosition=&AddVectorToPosition(
     Position=$Particles.Position
     Vector=&Multiply_Vector(
      A=&Add_Vector(
       A=&Multiply_Vector(
        A=&FN_GetSystemAxis(Axis:ENiagaraOrientationAxis=::XAxis, Flip Axis:NiagaraBool=false)
        B=&VectorFromFloat(Value=&RandomRangeFloat(Minimum=-1, Maximum=-1, Evaluation Type=::NewEnumerator1))
       )
       B=&Multiply_Vector(
        A=&ReturnMeshOrientationAxis(Mesh Orientation Axis=::YAxis, Mesh Orientation Quat=$Particles.MeshOrientation)
        B=&VectorFromFloat(Value=&RandomRangeFloat(Minimum=-1, Maximum=1, Evaluation Type=::NewEnumerator1))
       )
      )
      B=&VectorFromFloat(Value=1)
     )
    )) {isSetParameters=true, module=SetVariables_#10}
  ]
  rendererClasses [&NiagaraMeshRendererProperties, &NiagaraSpriteRendererProperties]
  renderers [
   {rendererClass=&NiagaraMeshRendererProperties, rendererIndex=0}
   {rendererClass=&NiagaraMeshRendererProperties, rendererIndex=1}
   {rendererClass=&NiagaraSpriteRendererProperties, rendererIndex=2}
   {rendererClass=&NiagaraMeshRendererProperties, rendererIndex=3}
  ]
 }
 MeshRipples002 {
  emitterUpdateScript [
   &EmitterState()
   &SpawnBurst_Instantaneous(Use Spawn Probability=false, UseLoopCountLimit=false)
   &SpawnBurst_Instantaneous(
    Spawn Probability=&MakeCustomFloatFromBool(
     Bool=&IsPlatformSetActive(New Platform Set=DataInterface() {propertyValues {Platforms {qualityLevelMask=-31}}})
     True Float=0
     False Float=1
    )
    Spawn Time=0.05
    Use Spawn Probability=true
    UseLoopCountLimit=false
   ) {module=SpawnBurst_Instantaneous002}
   &SpawnBurst_Instantaneous(Spawn Time=0.35, Use Spawn Probability=false, UseLoopCountLimit=false) {module=SpawnBurst_Instantaneous001}
   &SpawnBurst_Instantaneous(
    Spawn Probability=&MakeCustomFloatFromBool(
     Bool=&IsPlatformSetActive(New Platform Set=DataInterface() {propertyValues {Platforms {qualityLevelMask=-31}}})
     True Float=0
     False Float=1
    )
    Spawn Time=0.425
    Use Spawn Probability=true
    UseLoopCountLimit=false
   ) {module=SpawnBurst_Instantaneous003}
  ]
  particleSpawnScript [
   /SetVariables(Particles.Master Scale=0.75) {isSetParameters=true, module=SetVariables_#1}
   &InitializeParticle(
    Lifetime=0.325
    Mesh Scale Mode=::NewEnumerator3
    Mesh Uniform Scale=&Multiply_Float(A=1, B=$Particles.Master Scale)
    Position Offset=<10 -5 0>
    Position Offset Coordinate Space=::Local
    Sprite Rotation Mode=::NewEnumerator0
    Sprite Size Mode=::NewEnumerator3
    Uniform Sprite Size=&Multiply_Float(A=175, B=$Particles.Master Scale)
    UsePositionOffset=true
    Write Material Random=true
   )
   &AddVelocity(Velocity=<1 0 0>, Velocity Speed Scale=90)
   &InitialMeshOrientation(Mesh Orientation Mode=::NewEnumerator1, Rotation=<0 0.25 0>, Use Rotation Vector=true)
   /SetVariables(Particles.MeshOrientation=&MultiplyQuaternion(
     Quaternion A=&MakeQuatFromAxisAngle(
      Axis=&ReturnMeshOrientationAxis(Mesh Orientation Axis=::ZAxis, Mesh Orientation Quat=$Particles.MeshOrientation)
      AngleInDegrees=&RandomRangeFloat(Minimum=0, Maximum=360)
     )
     Quaternion B=$Particles.MeshOrientation
    )) {isSetParameters=true, module=SetVariables_#2}
   &CameraOffset()
  ]
  particleUpdateScript [
   &ParticleState()
   /SetVariables(Particles.RandScaleVector:Vector3f=&Lerp_Vector(
     A=<1 1 1>
     B=&RandomRangeVector(Minimum=<0.9 0.9 0.9>, Maximum=<1.1 1.1 1.1>, Evaluation Type=::NewEnumerator1)
     Alpha=0.75
    )) {isSetParameters=true, module=SetVariables_#3}
   &ScaleSpriteSize(
    Scale Sprite Size Mode=::NewEnumerator0
    Uniform Scale Factor=&FloatFromCurve(
     FloatCurve=DataInterface() {propertyValues {Curve {keys [<at 0 0.5 RCIM_Cubic>, {interpMode=RCIM_Cubic, leaveTangent=1.0401, time=1, value=0.7}]}}}
     Scale Curve=0.75
    )
   )
   &Drag(Drag=0.5, Use Linear Drag=true, Use Rotational Drag=true)
   &InheritSourceMovement(ApplyForce=false, ApplyPosition=true, ApplyPositionFalloff=false)
   &SolveForcesAndVelocity(Clamp Velocity=false, Limit Acceleration=false)
   &GenerateLocationEvent(Event Generation Enabled:NiagaraBool=$Transient.FirstFrame, Event Send Rate=120, Use Event Probability=false, UseEventDelay=false)
   &AlignSpriteToMeshOrientation(Mesh Orientation Relative Sprite Facing Vector:Vector3f=<0 0 1>, SetAlignment=true, SetFacing=true)
   /SetVariables(Particles.Mesh Orient Facing:Vector3f=$Particles.SpriteFacing) {isSetParameters=true, module=SetVariables_#4}
   &ScaleMeshSize(Scale Factor=&VectorFromFloat(Value=&FloatFromCurve(FloatCurve=DataInterface() {propertyValues {Curve {keys [
          {
           arriveTangent=0.3962
           arriveTangentWeight=0.4354
           interpMode=RCIM_Cubic
           leaveTangent=0.3962
           leaveTangentWeight=0.4354
           tangentMode=RCTM_User
           tangentWeightMode=RCTWM_WeightedBoth
           time=0.05
           value=0.5
          }
          {
           arriveTangent=0.1008
           arriveTangentWeight=0.5162
           interpMode=RCIM_Cubic
           leaveTangent=0.1008
           leaveTangentWeight=0.5162
           tangentMode=RCTM_User
           tangentWeightMode=RCTWM_WeightedBoth
           time=1
           value=0.77
          }
         ]}}})))
   &ScaleMeshSize(Scale Factor=$Particles.RandScaleVector) {module=ScaleMeshSize001}
   &ScaleColor(
    Scale Alpha=&FloatFromCurve(
     FloatCurve=DataInterface() {propertyValues {Curve {keys [
         {arriveTangent=-1.1111, interpMode=RCIM_Cubic, time=0, value=0}
         <at 0.1 1 RCIM_Cubic>
         {interpMode=RCIM_Cubic, leaveTangent=-3.1172, time=1, value=0}
        ]}}}
     Scale Curve=0.35
    )
    Scale RGB=&Lerp_Vector(
     A=&MakeVectorFromLinearColorRGB(Color=&ColorFromCurve(ColorCurve=DataInterface() {propertyValues {
         AlphaCurve {keys []}
         BlueCurve {keys [
           {interpMode=RCIM_Cubic, leaveTangent=3.2232, tangentMode=RCTM_Break, time=0, value=0.6688}
           {interpMode=RCIM_Cubic, tangentMode=RCTM_Break, time=0.1324, value=0.85}
           {arriveTangent=7.7104, interpMode=RCIM_Cubic, tangentMode=RCTM_Break, time=1, value=3}
          ]}
         GreenCurve {keys [
           <at 0 0.475 RCIM_Cubic>
           {
            arriveTangent=3.3365
            arriveTangentWeight=0.2322
            interpMode=RCIM_Cubic
            leaveTangent=-3.1167
            leaveTangentWeight=0.4094
            tangentMode=RCTM_Break
            tangentWeightMode=RCTWM_WeightedBoth
            time=0.1324
            value=0.7809
           }
           {
            arriveTangentWeight=0.6868
            interpMode=RCIM_Cubic
            leaveTangentWeight=0.6868
            tangentMode=RCTM_User
            tangentWeightMode=RCTWM_WeightedBoth
            time=1
            value=0.3063
           }
          ]}
         RedCurve {keys [
           <at 0 0.05 RCIM_Cubic>
           {
            arriveTangent=2.3842
            arriveTangentWeight=0.1724
            interpMode=RCIM_Cubic
            leaveTangent=-3.2096
            leaveTangentWeight=0.3729
            tangentMode=RCTM_Break
            tangentWeightMode=RCTWM_WeightedBoth
            time=0.1324
            value=0.2975
           }
           {
            arriveTangentWeight=0.6899
            interpMode=RCIM_Cubic
            leaveTangentWeight=0.6899
            tangentMode=RCTM_User
            tangentWeightMode=RCTWM_WeightedBoth
            time=1
            value=0.06
           }
          ]}
        }}))
     B=&MakeVectorFromLinearColorRGB(Color=&ColorFromCurve(ColorCurve=DataInterface() {propertyValues {
         AlphaCurve {keys []}
         BlueCurve {keys [
           {interpMode=RCIM_Cubic, leaveTangent=3.2232, tangentMode=RCTM_Break, time=0, value=0.6688}
           {interpMode=RCIM_Cubic, tangentMode=RCTM_Break, time=0.1324, value=0.85}
           {arriveTangent=7.7104, interpMode=RCIM_Cubic, tangentMode=RCTM_Break, time=1, value=3}
          ]}
         GreenCurve {keys [
           <at 0 0.475 RCIM_Cubic>
           {
            arriveTangent=3.3365
            arriveTangentWeight=0.2322
            interpMode=RCIM_Cubic
            leaveTangent=-3.1167
            leaveTangentWeight=0.4094
            tangentMode=RCTM_Break
            tangentWeightMode=RCTWM_WeightedBoth
            time=0.1324
            value=0.7809
           }
           {
            arriveTangentWeight=0.6868
            interpMode=RCIM_Cubic
            leaveTangentWeight=0.6868
            tangentMode=RCTM_User
            tangentWeightMode=RCTWM_WeightedBoth
            time=1
            value=0.2513
           }
          ]}
         RedCurve {keys [
           <at 0 0.05 RCIM_Cubic>
           {
            arriveTangent=2.3842
            arriveTangentWeight=0.1724
            interpMode=RCIM_Cubic
            leaveTangent=-3.2096
            leaveTangentWeight=0.3729
            tangentMode=RCTM_Break
            tangentWeightMode=RCTWM_WeightedBoth
            time=0.1324
            value=0.2975
           }
           {
            arriveTangentWeight=0.6899
            interpMode=RCIM_Cubic
            leaveTangentWeight=0.6899
            tangentMode=RCTM_User
            tangentWeightMode=RCTWM_WeightedBoth
            time=1
            value=0
           }
          ]}
        }}))
     Alpha=&RandomRangeFloat(Minimum=0, Maximum=1)
    )
    ScaleA=true
    ScaleRGB=true
   ) {enabled=false, module=ScaleColor001}
   &ScaleColor(
    Scale Alpha=&FloatFromCurve(
     FloatCurve=DataInterface() {propertyValues {Curve {keys [
         {arriveTangent=-1.1111, interpMode=RCIM_Cubic, time=0, value=0}
         <at 0.1 1 RCIM_Cubic>
         {arriveTangent=-0.9656, interpMode=RCIM_Cubic, leaveTangent=-0.9656, tangentMode=RCTM_User, time=1, value=0}
        ]}}}
     Scale Curve=0.35
    )
    Scale RGB=&MakeVectorFromLinearColorRGB(Color=&ColorFromCurve(ColorCurve=DataInterface() {propertyValues {
        AlphaCurve {keys []}
        BlueCurve {keys [
          {interpMode=RCIM_Cubic, leaveTangentWeight=0.8633, tangentMode=RCTM_Break, tangentWeightMode=RCTWM_WeightedBoth, time=0, value=0.5125}
          {arriveTangentWeight=0.5327, interpMode=RCIM_Cubic, tangentMode=RCTM_Break, tangentWeightMode=RCTWM_WeightedBoth, time=1, value=3}
         ]}
        GreenCurve {keys [
          {interpMode=RCIM_Cubic, leaveTangentWeight=0.8633, tangentMode=RCTM_Break, tangentWeightMode=RCTWM_WeightedBoth, time=0, value=0.0625}
          {
           arriveTangentWeight=0.5327
           interpMode=RCIM_Cubic
           leaveTangentWeight=0.6868
           tangentMode=RCTM_Break
           tangentWeightMode=RCTWM_WeightedBoth
           time=1
           value=0.3063
          }
         ]}
        RedCurve {keys [
          {interpMode=RCIM_Cubic, leaveTangentWeight=0.8633, tangentMode=RCTM_Break, tangentWeightMode=RCTWM_WeightedBoth, time=0, value=0.625}
          {
           arriveTangentWeight=0.5327
           interpMode=RCIM_Cubic
           leaveTangentWeight=0.6899
           tangentMode=RCTM_Break
           tangentWeightMode=RCTWM_WeightedBoth
           time=1
           value=0.06
          }
         ]}
       }}))
    ScaleA=true
    ScaleRGB=true
   )
   &DynamicMaterialParameters(
    Index 0 Param 4=&FloatFromCurve(
     FloatCurve=DataInterface() {propertyValues {Curve {keys [
         {
          arriveTangent=9.177
          arriveTangentWeight=0.6327
          interpMode=RCIM_Cubic
          leaveTangent=9.177
          leaveTangentWeight=0.6327
          tangentMode=RCTM_User
          tangentWeightMode=RCTWM_WeightedBoth
          time=0
          value=0.2
         }
         {
          arriveTangentWeight=0.6785
          interpMode=RCIM_Cubic
          leaveTangentWeight=0.6785
          tangentMode=RCTM_User
          tangentWeightMode=RCTWM_WeightedBoth
          time=1
          value=1
         }
        ]}}}
     Scale Curve=0.28
    )
    Module.Index0Param0Write=true
    Module.Index0Param1Write=true
    Module.Index0Param2Write=true
    Module.Index0Param3Write=true
   )
   &SpriteFacingAndAlignment(
    Facing Coordinate Space=::Local
    Sprite Alignment=&RandomVector(Vector Scale=1)
    Sprite Facing=&ReturnCameraVector(Camera Vectors=::NewEnumerator4, Coordinate Space=::Simulation, Origin Position=$Particles.Position)
    WriteAlignment=false
    WriteFacing=true
   )
   /SetVariables(Particles.Decal Color:LinearColor=&MultiplyLinearColor(
     Linear Color=$Particles.Color
     Scale Factor=&MakeFloat4FromVectorAndFloat("Vector (RGB)"=&VectorFromFloat(Value=2), "Float (Alpha)"=0.35)
    )) {isSetParameters=true, module=SetVariables_#5}
   /SetVariables(
    Particles.Refracl Color:LinearColor=&MultiplyLinearColor(
     Linear Color=$Particles.Color
     Scale Factor=&MakeFloat4FromVectorAndFloat("Vector (RGB)"=&VectorFromFloat(Value=5), "Float (Alpha)"=0)
    )
    Particles.refract dynamic:Vector4f=&MakeVector4(
     X=&Multiply_Float(
      A=&MakeFloatFromLinearColor(LinearColor=$Particles.Color, Channel:ENiagaraLinearColor_Channels=::NewEnumerator3)
      B=0.0025
     )
     Y=0
     Z=0
     W=0
    )
   ) {isSetParameters=true, module=SetVariables_#6}
   /SetVariables(Particles.RegularFacing:Vector3f=$Particles.SpriteFacing) {isSetParameters=true, module=SetVariables_#7}
   /SetVariables(Particles.NewFacing:Vector3f=&Lerp_Vector(
     A=$Particles.Mesh Orient Facing
     B=$Particles.RegularFacing
     Alpha=&FloatFromCurve(
      FloatCurve=DataInterface() {propertyValues {Curve {keys [<at -1 0 RCIM_Cubic>,<at 0 1 RCIM_Cubic>,<at 1 0 RCIM_Cubic>]}}}
      CurveIndex=&DotProduct(
       Vector A=&ReturnCameraVector(Camera Vectors=::NewEnumerator4, Coordinate Space=::Simulation, Origin Position=$Particles.Position)
       Vector B=$Engine.Owner.SystemXAxis
      )
     )
    )) {isSetParameters=true, module=SetVariables_#8}
   /SetVariables(Particles.NewFacing:Vector3f=$Particles.RegularFacing) {enabled=false, isSetParameters=true, module=SetVariables_#9}
   /SetVariables(Particles.VibratePosition:NiagaraPosition=&AddVectorToPosition(
     Position=$Particles.Position
     Vector=&Multiply_Vector(
      A=&Add_Vector(
       A=&Multiply_Vector(
        A=&FN_GetSystemAxis(Axis:ENiagaraOrientationAxis=::XAxis, Flip Axis:NiagaraBool=false)
        B=&VectorFromFloat(Value=&RandomRangeFloat(Minimum=-1, Maximum=-1, Evaluation Type=::NewEnumerator1))
       )
       B=&Multiply_Vector(
        A=&ReturnMeshOrientationAxis(Mesh Orientation Axis=::YAxis, Mesh Orientation Quat=$Particles.MeshOrientation)
        B=&VectorFromFloat(Value=&RandomRangeFloat(Minimum=-1, Maximum=1, Evaluation Type=::NewEnumerator1))
       )
      )
      B=&VectorFromFloat(Value=1)
     )
    )) {isSetParameters=true, module=SetVariables_#10}
  ]
  rendererClasses [&NiagaraMeshRendererProperties, &NiagaraSpriteRendererProperties]
  renderers [
   {rendererClass=&NiagaraMeshRendererProperties, rendererIndex=0}
   {rendererClass=&NiagaraMeshRendererProperties, rendererIndex=1}
   {rendererClass=&NiagaraSpriteRendererProperties, rendererIndex=2}
   {rendererClass=&NiagaraMeshRendererProperties, rendererIndex=3}
  ]
 }
}
summary {
 systemName=NS_FantasyEmote
 emitters [
  {
   bEnabled=false
   emitterName=MeshRipples
   rendererClasses [&NiagaraMeshRendererProperties, &NiagaraSpriteRendererProperties]
  }
  {emitterName=Air_glow, rendererClasses [&NiagaraSpriteRendererProperties]}
  {
   bEnabled=false
   emitterName=MeshRipples001
   rendererClasses [&NiagaraMeshRendererProperties, &NiagaraSpriteRendererProperties]
  }
  {emitterName=MeshRipples002, rendererClasses [&NiagaraMeshRendererProperties, &NiagaraSpriteRendererProperties]}
 ]
 userVariables []
}
